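; Throwable Launch Pad trajectory preview that shows a grid-snapped building ghost where the pad will land.
; HandleGridSnapUpdates snaps the arc endpoint to the building grid, flags blacklisted hits and shifts for ramps.
; On cell, yaw or hotfix-gated endpoint changes, it validates placement through shared Launchpad helpers, orients to instigator facing and picks the floor-variant mesh.
; Visibility follows trajectory validity, with extra touch-input conditions; several local ramp/floor helpers go uncalled here.
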(event Trajectory|EventSetTrajectorySpline
  (params (SplinePoints :type Array%20of%20Vectors) (SplineTangents :type Array%20of%20Vectors))
  (|CallParentFunction SplinePoints SplineTangents))

(fn CheckforBlacklistedTags
  (params (Location :type Vector))
  (bind _outhits (Collision|MultiSphereTraceForObjects Location Location 1e+01 (Utilities|Array|MakeArray "ObjectTypeQuery2")))
  (if (Utilities|Array|IsEmpty _outhits)
    (return false)
    (else
      (for _array_element _outhits
        (Utilities|IsValid (.HitActor (Collision|BreakHitResult _array_element))
          (:Is%20Valid
            (bind _returnvalue (GameplayTags|HasAnyMatchingGameplayTags(Message) (.HitActor (Collision|BreakHitResult _array_element)) (LocalVariables|GetBlacklisted%20Tags)))
            (if _returnvalue
              (return true)))))
      (return false))))

(fn CheckforRamp
  (params (Location :type Vector))
  (bind _outhits (Collision|MultiSphereTraceByChannel Location Location 5e+01 "TraceTypeQuery10"))
  (for _array_element _outhits
    (Utilities|IsValid (.HitActor (Collision|BreakHitResult _array_element))
      (:Is%20Valid
        (bind _returnvalue (GameplayTags|HasAnyMatchingGameplayTags(Message) (.HitActor (Collision|BreakHitResult _array_element)) (Variables|RampRules|GetRamp%20Gameplay%20Tags)))
        (if _returnvalue
          (bind _output_get (LocalVariables|SetHit%20Actor (.HitActor (Collision|BreakHitResult _array_element))))
          (bind _asbuilding_stairs (Utilities|Casting|CastToBuildingStairs _output_get)
            (:then
              (if (== (Building|GetNavigationLinksClass _asbuilding_stairs false) (LocalVariables|GetStairs%20Nav%20Links%20Class))
                _))
            (:CastFailed
              (bind _asbuilding_roof (Utilities|Casting|CastToBuildingRoof (LocalVariables|GetHit%20Actor)))
              (if (== (Building|GetNavigationLinksClass _asbuilding_roof false) (LocalVariables|GetRoof%20Nav%20Links%20Class))
                _)))
          (return true (LocalVariables|GetHit%20Actor))))))
  (return false 0))

(fn CheckHitsForRamp
  (params (Hit%20Results :type Array%20of%20Hit%20Result%20Structures))
  (if (Utilities|Array|IsEmpty Hit%20Results)
    (return false 0)
    (else
      (for _array_element Hit%20Results
        (if (Default|LaunchpadRampCheck _array_element)
          (return (Default|LaunchpadRampCheck _array_element) (.Ramp%20Actor (Default|LaunchpadRampCheck _array_element)))))
      (return false 0))))

(fn FloorTrace
  (bind _outhits (Collision|MultiBoxTraceByChannel (+ (Variables|Default|GetDoorTraceOffsetLow) (Variables|Default|GetModified%20Grid%20Snap%20Position)) (Math|Vector|vector+vector (Variables|Default|GetDoorTraceOffsetLow) (Variables|Default|GetModified%20Grid%20Snap%20Position) "0,0,-15.000000") "128.000000,128.000000, 0" "0, 0, 0" "TraceTypeQuery10"))
  (for _array_element _outhits
    (bind _returnvalue (GameplayTags|HasAnyMatchingGameplayTags(Message) (.HitActor (Collision|BreakHitResult _array_element)) (Variables|Default|GetFloor%20Gameplay%20Tags)))
    (if _returnvalue
      (return true)))
  (return false))

(fn HandleRamps
  (params (Ramp%20Actor :type Actor%20Object%20Reference) (Target%20Location :type Vector))
  (Utilities|IsValid Ramp%20Actor
    (:Is%20Valid
      (if (not (RampRules|RampUndersideCheck Ramp%20Actor Target%20Location))
        (Variables|Default|SetModified%20Grid%20Snap%20Position (+ (Variables|Default|GetModified%20Grid%20Snap%20Position) (Utilities|Map|Find (Variables|RampRules|GetRamp%20Rule%20Build%20Offsets) (RampRules|RoundtoNearestRotation (Math|Float|Truncate (.yaw (Transformation|GetActorRotation Ramp%20Actor)))))))))))

(fn HandleGridSnapUpdates
  (params (NewLocation :type Vector :default "0, 0, 0") (TrajectoryHitResult :type Hit%20Result%20Structure))
  ;
  (Variables|RampRules|SetOld%20Last%20Spline%20Point%20Coord (Variables|RampRules|GetLast%20Spline%20Point%20Coord))
  (Variables|RampRules|SetLast%20Spline%20Point%20Coord NewLocation)
  ;
  (bind _output_get (Variables|Default|SetGridSnapPosition (Fortnite|SnaptoFortBuildingGrid (Variables|RampRules|GetLast%20Spline%20Point%20Coord) (Math|Vector|MakeVector 0.0 0.0 (Variables|SnaptoGrid|GetGridZOffset)) (Math|Vector|MakeVector 0.0 (Variables|SnaptoGrid|GetGridYOffset) (* (Variables|SnaptoGrid|GetGridZOffset) -1.0)))))
  (Variables|Default|SetModified%20Grid%20Snap%20Position _output_get)
  ;[1 0.9531 0.4163] Check Impact of Target for valid Ramp Actors, and if so, set Ramp Rule for later use
  (bind (_outhits _returnvalue) (Collision|MultiSphereTraceByChannel :Start (Variables|RampRules|GetLast%20Spline%20Point%20Coord) :End (Variables|RampRules|GetLast%20Spline%20Point%20Coord) :Radius 5e+01 :TraceChannel "TraceTypeQuery10" :bTraceComplex false :DrawDebugType "None" :bIgnoreSelf true :TraceColor "(R=1.000000,G=0.933392,B=0.000000,A=1.000000)" :TraceHitColor "(R=1.000000,G=0.000000,B=0.729970,A=1.000000)" :DrawTime 5.0))
  (Variables|Default|SetBlacklisted%20Tag%20Found (CallFunction|CheckForBlacklistedTags (Variables|RampRules|GetLast%20Spline%20Point%20Coord)))
  (if _returnvalue
    (if (CallFunction|CheckHitsforRamp _outhits)
      (Variables|RampRules|SetRamp%20Rule%20Applies true)
      (bind _new_grid_snapped_position (Class|LaunchpadFunctionLibrary|LaunchpadRampHandling (.Ramp%20Actor (CallFunction|CheckHitsforRamp _outhits)) (Variables|RampRules|GetLast%20Spline%20Point%20Coord) self (Variables|Default|GetModified%20Grid%20Snap%20Position) "0, 0,-192.000000"))
      (Variables|Default|SetModified%20Grid%20Snap%20Position _new_grid_snapped_position)
      (else
        (Variables|RampRules|SetRamp%20Rule%20Applies false)))
    (else
      (Variables|RampRules|SetRamp%20Rule%20Applies false)))
  (bind (_outcardinaldirection _outyaw) (Class|AthenaConsumablesSharedFunctionLibrary|GetActorForwardCardinalDirection (Game|GetInstigator)))
  (LocalVariables|SetYawOffset _outyaw)
  (Utilities|FlowControl|Sequence
    (:then_0
      (if (Math|Vector|NotEqual(Vector) (Variables|Default|GetGridSnapPosition) (Variables|Default|GetLastGridSnapPosition) 0.0001)
        (Audio|PlaySound2D (Variables|Default|GetGridSnapSound))
        (if (or (not (Math|Float|NearlyEqual(Float) (LocalVariables|GetYawOffset) (Variables|Default|GetLastYawOffset) 0.0001)) (and (Utilities|Operators|NotEqual(!=) (Variables|RampRules|GetLast%20Spline%20Point%20Coord) (Variables|RampRules|GetOld%20Last%20Spline%20Point%20Coord) 0.0001) (ScalableFloat|GetValueasBool (LocalVariables|GetHotfix%20-%20Enable%20Update%20On%20Trajectory%20End%20Loc%20Changed))))
          (Variables|Default|SetLastGridSnapPosition (Variables|Default|GetGridSnapPosition))
          (if (Variables|RampRules|GetRamp%20Rule%20Applies)
            (bind _output_get_2 (Variables|Default|SetGridSnapPosition (Variables|Default|GetModified%20Grid%20Snap%20Position)))
            (Variables|Default|SetRamp%20Floating%20Check (Class|LaunchpadFunctionLibrary|LaunchpadRampFloatCheck self (+ _output_get_2 "0,0,-192.000000")))
            (else
              (Variables|Default|SetRamp%20Floating%20Check false)))
          (Variables|Default|SetLastYawOffset (LocalVariables|GetYawOffset))
          (bind (_is_valid _floor_variant _new_center_position) (Default|LaunchpadValidatePlacement self (Variables|Default|GetModified%20Grid%20Snap%20Position) "0, 0, 0" (Variables|GetTrajectoryHitResult) "0.000000,0.000000,320.000000" "0.000000,0.000000,-192.000000" (LocalVariables|GetEmpty%20Array)))
          (Variables|Default|SetIs%20Placement%20Valid _is_valid)
          (Variables|Default|SetFloor%20Variant _floor_variant)
          (Variables|Default|SetModified%20Grid%20Snap%20Position _new_center_position)
          (Transformation|SetWorldLocationAndRotation (Variables|Default|GetBuilding) (Variables|Default|GetModified%20Grid%20Snap%20Position) (Math|Rotator|MakeRotator 0.0 0.0 (LocalVariables|GetYawOffset)) false true)
          (Components|StaticMesh|SetStaticMesh (Variables|Default|GetSM_BuildingPreview) (Utilities|Select :Index (Variables|Default|GetFloor%20Variant) :NewEnumerator0 (Variables|Default|GetBuildingPreviewMeshUnderground) :NewEnumerator1 (Variables|Default|GetBuildingPreviewMeshFloorless) :NewEnumerator2 (Variables|Default|GetBuildingPreviewMesh))))))
    (:then_1
      (if (or (Math|Vector|NotEqual(Vector) (Variables|Default|GetGridSnapPosition) (Variables|Default|GetLastGridSnapPosition) 0.0001) (Math|Float|NearlyEqual(Float) (LocalVariables|GetYawOffset) (Variables|Default|GetLastYawOffset) 0.0001))
        (if (TouchPlatform|IsTouchInputPlatform)
          (bind _asfort_player_pawn (Utilities|Casting|CastToFortPlayerPawn (Game|GetInstigator)))
          (Rendering|SetVisibility (Variables|Default|GetSM_BuildingPreview) (Math|Boolean|ANDBoolean (GameplayTags|HasMatchingGameplayTag _asfort_player_pawn "(TagName=\"Gameplay.Action.Player.InWindup\")") (not (Class|AthenaHUDContext|UseTaptoShoot (BlueprintContext|GetAthenaHUDContext))) (Variables|Default|GetbIsTrajectoryValid) (not (Variables|Default|GetbWaitFirstTrajectoryValid))) true)
          (else
            (Rendering|SetVisibility (Variables|Default|GetSM_BuildingPreview) (and (Variables|Default|GetbIsTrajectoryValid) (not (Variables|Default|GetbWaitFirstTrajectoryValid))) true)))))))

(fn InstigatorStandingOnRamp
  (params (Ramp%20Actor :type Actor%20Object%20Reference))
  (bind _outhits (Collision|MultiSphereTraceByChannel (- (Transformation|GetActorLocation (Variables|Actor|GetInstigator)) "0,0,75.000000") (- (Transformation|GetActorLocation (Variables|Actor|GetInstigator)) "0,0,75.000000") 25.0 "TraceTypeQuery10"))
  (for _array_element _outhits
    (if (== Ramp%20Actor (.HitActor (Collision|BreakHitResult _array_element)))
      (return true)))
  (return false))

(fn MultipleSphereTraces
  (params (Location :type Vector))
  (bind (_outhits _returnvalue) (Collision|MultiSphereTraceByChannel Location Location 5e+01 "TraceTypeQuery10" false (LocalVariables|GetActors%20to%20Ignore) "None" true "(R=1.000000,G=0.933392,B=0.000000,A=1.000000)" "(R=1.000000,G=0.000000,B=0.729970,A=1.000000)"))
  (if _returnvalue
    (for _array_element _outhits
      (if (Default|LaunchpadRampCheck _array_element)
        (Utilities|Array|Add (LocalVariables|GetStored%20Hits) _array_element)
        (Utilities|Array|Add (LocalVariables|GetActors%20to%20Ignore) (.Ramp%20Actor (Default|LaunchpadRampCheck _array_element))))))
  (return 0))

(fn RampUndersideCheck
  (params (Ramp%20Actor :type Actor%20Object%20Reference) (Ramp%20Impact%20Location :type Vector))
  (LocalVariables|SetCached%20Ramp%20Actor Ramp%20Actor)
  (LocalVariables|SetCached%20Impact%20Location Ramp%20Impact%20Location)
  (Utilities|FlowControl|Sequence
    (:then_0
      ;[1 0.8041 0.1602] Above Ramp Check via Line Trace
      (bind _outhits (Collision|MultiLineTraceByChannel (+ (LocalVariables|GetCached%20Impact%20Location) (LocalVariables|GetAbove%20Line%20Start%20Offset)) (LocalVariables|GetCached%20Impact%20Location) "TraceTypeQuery10"))
      (bind _array_element (Utilities|Array|ForEachLoopwithBreak _outhits)
        (:LoopBody
          (if (== (LocalVariables|GetCached%20Ramp%20Actor) (.HitActor (Collision|BreakHitResult _array_element)))
            (LocalVariables|SetAbove%20Dist (.Distance (Collision|BreakHitResult _array_element)))))))
    (:then_1
      ;[1 0.9009 0.1521] Below Ramp Check via Line Trace
      (bind _outhits_2 (Collision|MultiLineTraceByChannel (+ (LocalVariables|GetCached%20Impact%20Location) (LocalVariables|GetBelow%20Line%20Start%20Offset)) (LocalVariables|GetCached%20Impact%20Location) "TraceTypeQuery10"))
      (bind _array_element_2 (Utilities|Array|ForEachLoopwithBreak _outhits_2)
        (:LoopBody
          (if (== (LocalVariables|GetCached%20Ramp%20Actor) (.HitActor (Collision|BreakHitResult _array_element_2)))
            (LocalVariables|SetBelow%20Dist (.Distance (Collision|BreakHitResult _array_element_2)))))))
    (:then_2
      ;[1 0.8298 0.122] Returns Bool for which line trace is longer, indicating which side
      ; of ramp preview hit
      (return (>= (LocalVariables|GetAbove%20Dist) (LocalVariables|GetBelow%20Dist))))))

(fn RoundToNearestRotation
  (params (Angle :type Integer))
  (bind _output_get (LocalVariables|SetIncoming%20Angle Angle))
  (if (Math|Integer|InRange(Integer) _output_get -5 5)
    (return 0)
    (elif (Math|Integer|InRange(Integer) (LocalVariables|GetIncoming%20Angle) 85 95)
      (return 90)
      (elif (Math|Integer|InRange(Integer) (LocalVariables|GetIncoming%20Angle) 175 185)
        (return 180)
        (elif (Math|Integer|InRange(Integer) (LocalVariables|GetIncoming%20Angle) -95 -85)
          (return -90))))))

(fn ConstructionScript
  (|CallParentFunction))
