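; Parent throwable projectile for the Super Launch Pad, which builds a launch pad structure where it lands.
; It bounces off pawns and props, and on stop snaps to the building grid with ramp adjustments, facing the thrower's cardinal direction.
; After exploding, it validates placement and hostile overlap, clears props on a timer, then spawns a normal, floorless or raised basement builder.
; Failed placements, including after bounces, toss the item back as a pickup.
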(event EventBeginPlay
  ;[1 1 1] Lookup build delay row. Determine which cardinal direction the tower should face.
  (bind _returnvalue (Audio|SpawnSoundAttached :Sound (Variables|Cosmetic|GetInAirLoopSound) :AttachToComponent (Variables|Collision|GetCapsuleComponent) :AttachPointName "None" :LocationType "KeepRelativeOffset" :bStopWhenAttachedToDestroyed true :VolumeMultiplier 1.0 :PitchMultiplier 1.0 :StartTime 0.0 :bAutoDestroy true))
  (bind _output_get (Variables|Components|SetInAirLoopComponent _returnvalue))
  (Audio|Components|Audio|FadeIn _output_get 0.5 1.0 0.0 "Linear" :!class /Script/Engine.AudioComponent)
  (CallFunction|CalculateCardinalDirection))

(event EventEndPlay
  (params (EndPlayReason :type EEndPlayReason%20Enum :default "Destroyed"))
  (bind _buildingstructuralsupportsystem (Building|GetBuildingStructuralSupportSystem))
  ;[1 1 1] Destroy any remaining actors in our destroy list if needed
  (switch Utilities|FlowControl|Switch|SwitchonEEndPlayReason EndPlayReason
    (:Destroyed
      (Building|StartActorRemovalBatch _buildingstructuralsupportsystem)
      (for _array_element (Variables|Tower|GetBuildingActorsToDestroy)
        (Actor|DestroyActor _array_element))
      (Building|StopActorRemovalBatch _buildingstructuralsupportsystem))))

(event Collision|EventHit
  (params (MyComp :type Primitive%20Component%20Object%20Reference) (Other :type Actor%20Object%20Reference) (OtherComp :type Primitive%20Component%20Object%20Reference) (bSelfMoved :type Boolean) (HitLocation :type Vector :default "0, 0, 0") (HitNormal :type Vector :default "0, 0, 0") (NormalImpulse :type Vector :default "0, 0, 0") (Hit :type Hit%20Result%20Structure%20%28by%20ref%29))
  ;[1 1 1] Bounce off of pawns and walls
  (bind _bshouldbounce (CallFunction|ShouldBounceOffActor Other OtherComp))
  (if _bshouldbounce
    (Variables|Default|SetBounced true)
    (CallFunction|HandleBounce "0, 0, 0")
    (Utilities|FlowControl|DoOnce false
      (:Completed
        (Utilities|FlowControl|Delay (Variables|Cosmetic|GetBounceTimeout))))
    (else
      _))
  (Utilities|FlowControl|DoOnce false
    (:Completed
      (Game|Components|ProjectileMovement|StopSimulating (Variables|Movement|GetProjectileMovementComponent) Hit))))

(event Custom|Handle%20Bounce
  (params (Hit%20Location :type Vector))
  (bind _lastbouncelocation (Variables|Cosmetic|GetLastBounceLocation))
  (bind _bouncesound (Variables|Cosmetic|GetBounceSound))
  ;[1 1 1] Play audio feedback, etc when bouncing
  (if (and (> (Math|Vector|VectorLengthSquared (- Hit%20Location (Variables|Cosmetic|GetLastBounceLocation))) (Math|Float|Square (Variables|Cosmetic|GetMinBounceDistance))) (> (- (Utilities|Time|GetTimeSeconds) (Variables|Cosmetic|GetLastBounceTime)) (Variables|Cosmetic|GetMinTimeBetweenBounces)))
    (Variables|Cosmetic|SetLastBounceLocation Hit%20Location)
    (Variables|Cosmetic|SetLastBounceTime (Utilities|Time|GetTimeSeconds))
    (Audio|PlaySoundatLocation _bouncesound _lastbouncelocation)
    (Fortnite|ShowSoundIndicatorLocalforTeam :Location _lastbouncelocation :MaxAudibleDistance (Variables|/Script/Engine.SoundBase|GetMaxDistance _bouncesound) :IndicatorType "Gunshot" :Team (Team|GetActorTeam self) :Affiliations (Utilities|Array|MakeArray "Hostile" "Neutral") :Instigator self :Tint "(R=0.000000,G=0.000000,B=0.000000,A=0.000000)" :AttenuationMaxRangeMultiplier 1.0 :AttenuationIntensityMultiplier 1.0)))

(event EventOnStop
  (params (Hit :type Hit%20Result%20Structure%20%28by%20ref%29))
  (bind (_bblockinghit _binitialoverlap _time _distance _location _impactpoint _normal _impactnormal _physmat _hitactor _hitcomponent) (Collision|BreakHitResult Hit))
  (bind (_bblockinghit_2 _binitialoverlap_2 _time_2 _distance_2 _location_2 _impactpoint_2) (Collision|BreakHitResult (Variables|Default|GetStopped%20Hit)))
  (bind (_ramp_rules_apply _ramp_actor) (Default|LaunchpadRampCheck (Variables|Default|GetStopped%20Hit)))
  (bind _stopped_hit (Variables|Default|GetStopped%20Hit))
  ;[1 1 1] Initializing variables, stop rotation, and snap projectile to grid before "exploding"
  (bind _inairloopcomponent (Variables|Components|GetInAirLoopComponent)
    (:then
      (Audio|Components|Audio|FadeOut _inairloopcomponent 0.25 0.0 "Linear" :!class /Script/Engine.AudioComponent))
    (:else
      _))
  (bind _boverlapping (Default|IsOverlappingDeployableBlockVolume self (Variables|Default|GetBlockingAreaCheckRadius) (Transformation|GetActorLocation)))
  (if (Math|Boolean|ORBoolean (CallFunction|CheckforBlacklistedTags _hitactor) (Variables|Default|GetBounced) _boverlapping (Utilities|Enum|Equal(Enum) (Collision|GetCollisionObjectType _hitcomponent) "ECC_GameTraceChannel18"))
    ;;[1 0 0] Failure
    (CallFunction|RejectBuilding true)
    (else
      (Variables|Default|SetStopped%20Hit Hit)
      (Rendering|SetVisibility (Variables|Components|GetMesh) false false)
      (CallFunction|SetBaseDestination _stopped_hit)
      (if _ramp_rules_apply
        (bind _new_grid_snapped_position (Class|LaunchpadFunctionLibrary|LaunchpadRampHandling _ramp_actor _impactpoint_2 self (Variables|BuildingAndClearing|GetBaseDestination) "0.000000,256.000000, 0"))
        (bind _output_get (Variables|BuildingAndClearing|SetBaseDestination _new_grid_snapped_position))
        (if (Class|LaunchpadFunctionLibrary|LaunchpadRampFloatCheck self (+ _output_get (Variables|Default|GetBottom%20Center%20Cell%20Offset)))
          ;;[1 0 0] Failure - Floating ramp
          (CallFunction|RejectBuilding false)
          (else
            _)))
      ;;[1 1 1] Center of Floor, offset from Base Destination
      (Variables|BuildingandClearing|SetCenterDestination (+ (Variables|BuildingAndClearing|GetBaseDestination) (Variables|Default|GetBottom%20Center%20Cell%20Offset)))
      (Components|Movement|SetUpdatedComponent (Variables|Components|GetRotatingMovement) 0)
      (Transformation|SetActorTransform self (Math|Transform|MakeTransform (+ (Variables|BuildingAndClearing|GetBaseDestination) "0,256.000000,576.000000") (Math|Rotator|MakeRotator 0.0 0.0 (Utilities|Map|Find (Variables|Orientation|GetOrientationYawOffsets) (Variables|Orientation|GetCardinalDirection)))) false false)
      (Components|Activation|Deactivate (Variables|Default|GetN_Trail))
      (Delivery|DoExplosion))))

(event EventOnExploded
  (params (HitActors :type Array%20of%20Actor%20Object%20References) (HitResults :type Array%20of%20Hit%20Result%20Structures))
  ;[0.5157 1 0.439] On Exploded Trace to see if the launchpad intersects with anything
  (bind _outactors (Collision|BoxOverlapActors (+ (Variables|BuildingandClearing|GetCenterDestination) "0,0,30.000000") "120.000000,120.000000,30.000000" (Utilities|Array|MakeArray "ObjectTypeQuery1" "ObjectTypeQuery2") :!class /Script/Engine.KismetSystemLibrary))
  (Networking|CanExecuteCosmeticEvents
    (:False
      (CallFunction|HandleProps _outactors)
      ;[1 0.3503 0.8798] Evaluate Desired Launchpad to Spawn
      (bind (_is_valid _floor_variant _new_center_position) (Default|LaunchpadValidatePlacement self (Variables|BuildingandClearing|GetCenterDestination) "0, 0, 0" (Variables|Default|GetStopped%20Hit) (Variables|BuildingandClearing|GetTraceTopOffset) "0, 0, 0" (Variables|Default|GetEmpty%20Array)))
      (if _is_valid
        (Variables|Default|SetFloor%20Variant _floor_variant)
        (Variables|BuildingandClearing|SetCenterDestination _new_center_position)
        ;;[1 1 1] Check for enemy structures building a Launchpad would destroy
        (bind _is_hostile (Class|LaunchpadFunctionLibrary|LaunchpadAffiliationCheck self (Fortnite|GetFortPlayerControllerfromActor (Variables|Actor|GetInstigator)) (+ (Variables|BuildingandClearing|GetCenterDestination) "0,0,192.000000") (Variables|Default|GetFloor%20Variant)))
        (if _is_hostile
          ;;[1 0 0] Failure - Overlapping hostile actors
          (CallFunction|RejectBuilding false)
          (else
            ;;[0 1 0] Success - Launchpad can be spawned!
            (CallFunction|PlayExplodeFX (Variables|BuildingandClearing|GetCenterDestination))
            (CallFunction|BeginDestroyingActors)))
        (else
          ;;[1 0 0] Failure - Spawn location is invalid
          (CallFunction|RejectBuilding false))))))

(event Custom|Play%20Explode%20FX
  (params (Center%20Destination :type Vector))
  (bind _returnvalue (Transformation|GetActorLocation))
  (bind _returnvalue_2 (- (+ Center%20Destination "0,0.000000,384.000000") "0, 0,500.000000"))
  (bind _gridexplosionoffset (Variables|Cosmetic|GetGridExplosionOffset))
  (bind _returnvalue_3 (+ _returnvalue _gridexplosionoffset))
  (bind _returnvalue_4 (+ Center%20Destination "0,0.000000,384.000000"))
  (bind _cue_distantsound (Variables|Cosmetic|GetCue_DistantSound))
  ;[1 1 1] Play Explode FX - Emitter, Sound
  (bind _returnvalue_5 (Niagara|SpawnSystematLocation (Variables|Cosmetic|GetP_BuildBuilding) _returnvalue_4))
  (Variables|Cosmetic|SetBuild%20Particle _returnvalue_5)
  (Camera|PlayWorldCameraShake "/Game/Athena/Items/Gameplay/BackPacks/CarminePack/FX/B_Ground_CameraShake_Heavy.B_Ground_CameraShake_Heavy_C" _returnvalue_2 8e+02 4.5e+03 1.8)
  (Fortnite|PlayLocalForceFeedbackatLocationMulti "/Game/Weapons/ForceFeedback/Explosions/FF_Generic_Explosion_Near.FF_Generic_Explosion_Near" "/Game/Weapons/ForceFeedback/Explosions/FF_Generic_Explosion_Far.FF_Generic_Explosion_Far" _returnvalue_2 1e+03 2.5e+03)
  (bind _inairloopcomponent (Variables|Components|GetInAirLoopComponent)
    (:then
      (Audio|Components|Audio|FadeOut _inairloopcomponent 0.2 0.0 "Linear" :!class /Script/Engine.AudioComponent))
    (:else
      _))
  (if (Pawn|IsLocallyControlled (Game|GetInstigator))
    (Audio|PlaySoundatLocation (Variables|Cosmetic|GetCue_CloseSound) _returnvalue_3 "0, 0, 0" 1.0 1.4) ; Play Close Explosion
    (else
      (Audio|PlaySoundatLocation _cue_distantsound _returnvalue_3 "0, 0, 0" 1.0 1.4) ; Play Distant Explosion
    ))
  (Fortnite|ShowSoundIndicatorLocalforTeam :Location _returnvalue_3 :MaxAudibleDistance (Variables|/Script/Engine.SoundBase|GetMaxDistance _cue_distantsound) :IndicatorType "Gunshot" :Team (Team|GetActorTeam self) :Affiliations (Utilities|Array|MakeArray "Hostile" "Neutral") :Instigator self :Tint "(R=0.000000,G=0.000000,B=0.000000,A=0.000000)" :AttenuationMaxRangeMultiplier 1.0 :AttenuationIntensityMultiplier 1.0))

(event Custom|Begin%20Destroying%20Actors
  ;[0.3646 1 1] Begin Destroying Actors -- loop a timer deleting an actor every frame or so
  (bind _returnvalue (Utilities|Time|SetTimerbyEvent (delegate Custom|Delete%20another%20actor) 0.03 true)))

(event Custom|Delete%20another%20actor
  (bind _buildingactorstodestroy (Variables|Tower|GetBuildingActorsToDestroy))
  ;[0.3646 1 1] Begin Destroying Actors -- loop a timer deleting an actor every frame or so
  (if (> (Utilities|Array|Length _buildingactorstodestroy) 0)
    (bind (_array_element _array_index) (Utilities|Array|ForEachLoopwithBreak _buildingactorstodestroy)
      (:LoopBody
        (Utilities|IsValid _array_element
          (:Is%20Valid
            (if (Variables|/Script/Engine.Actor|GetbCanBeDamaged _array_element)
              (bind _returnvalue (GameplayTags|HasAnyMatchingGameplayTags(Message) _array_element (Variables|Default|GetFloor%20Gameplay%20Tag)))
              (if (not _returnvalue)
                (Actor|DestroyActor _array_element)
                (Utilities|Array|RemoveIndex _buildingactorstodestroy _array_index))))
          (:Is%20Not%20Valid
            _))))
    (else
      (Utilities|Time|ClearandInvalidateTimerbyHandle (stale Custom|Begin%20Destroying%20Actors _returnvalue))
      (CallFunction|SpawnBuilding))))

(event Custom|Spawn%20Building
  ;[1 0.9211 0.5718] Spawn Building
  (switch Utilities|FlowControl|Switch|SwitchonEnum_LaunchpadSpawnVariants (Variables|Default|GetFloor%20Variant)
    (:NewEnumerator0
      ;;[1 0.8219 0] Launchpad would be embedded in terrain, so spawn it up on a platform instead
      (Tower|ClearGridCells (Variables|Default|GetClearedCells_Underground) true (Variables|BuildingAndClearing|GetBaseDestination) (Utilities|Map|Find (Variables|Orientation|GetOrientationYawOffsets) (Variables|Orientation|GetCardinalDirection)))
      (CallFunction|CallDynamicBuilder (Variables|Default|GetBuilderClassBasement) (Variables|BuildingAndClearing|GetBaseDestination))
      (Variables|BuildingAndClearing|SetBaseDestination (+ (Variables|BuildingAndClearing|GetBaseDestination) "0, 0,384.000000")))
    (:NewEnumerator1
      ;;[1 0.7555 0] Spawn Floorless Launchpad
      (CallFunction|CallDynamicBuilder (Variables|Default|GetBuilderClassFloorless) (Variables|BuildingAndClearing|GetBaseDestination)))
    (:NewEnumerator2
      ;;[1 0.7555 0] Spawn Normal Builder
      (CallFunction|CallDynamicBuilder (Variables|Default|GetBuilderClass) (Variables|BuildingAndClearing|GetBaseDestination)))))

(fn CalculateCardinalDirection
  (bind _returnvalue (Game|GetInstigator))
  (bind _returnvalue_2 (Transformation|GetActorForwardVector _returnvalue))
  (bind _x (.x _returnvalue_2))
  (bind _y (.y _returnvalue_2))
  (if (>= (Math|Float|Absolute(Float) _x) (Math|Float|Absolute(Float) _y))
    (if (>= _x 0.0)
      (Variables|Orientation|SetCardinalDirection "NewEnumerator0")
      (else
        (Variables|Orientation|SetCardinalDirection "NewEnumerator1")))
    (elif (>= _y 0.0)
      (Variables|Orientation|SetCardinalDirection "NewEnumerator2")
      (else
        (Variables|Orientation|SetCardinalDirection "NewEnumerator3")))))

(fn CallDynamicBuilder
  (params (Class :type Actor%20Class%20Reference) (Location :type Vector))
  (bind _returnvalue (Game|SpawnBuildingActorfromClass Class (Math|Transform|MakeTransform Location (Math|Rotator|MakeRotator 0.0 0.0 (Utilities|Map|Find (Variables|Orientation|GetOrientationYawOffsets) (Variables|Orientation|GetCardinalDirection)))) 0 false 0))
  (bind _asbuilding_actor (Utilities|Casting|CastToBuildingActor _returnvalue))
  (bind _asfort_player_controller (Utilities|Casting|CastToFortPlayerController (Game|GetInstigatorController)))
  (Building|InitializeKismetSpawnedBuildingActor _asbuilding_actor 0 _asfort_player_controller true 0 false))

(fn CanDestroyObject
  (params (Object :type Actor%20Object%20Reference))
  (bind _object (Variables|GetObject))
  (bind _bcanbedamaged (Variables|/Script/Engine.Actor|GetbCanBeDamaged _object))
  (bind _asbuilding_actor (Utilities|Casting|CastToBuildingActor Object)
    (:then
      (LocalVariables|SetBuilding _asbuilding_actor)
      (bind _returnvalue (CallFunction|GetOwnerWeapon))
      (Utilities|IsValid _returnvalue
        (:Is%20Valid
          (Utilities|IsValid (Actor|GetOwner _returnvalue)
            (:Is%20Valid
              (bind _aspawn (Utilities|Casting|CastToPawn (Actor|GetOwner _returnvalue))
                (:then
                  (bind _returnvalue_2 (Class|BuildingActor|CanDestroyBuildingActor (LocalVariables|GetBuilding) (Pawn|GetController _aspawn) _aspawn))
                  (return _returnvalue_2))
                (:CastFailed
                  (return _bcanbedamaged))))
            (:Is%20Not%20Valid
              (return _bcanbedamaged))))
        (:Is%20Not%20Valid
          (return _bcanbedamaged))))
    (:CastFailed
      (return (Variables|/Script/Engine.Actor|GetbCanBeDamaged (Variables|GetObject))))))

(fn CheckForBlacklistedTags
  (params (Hit%20Actor :type Actor%20Object%20Reference))
  (Utilities|IsValid Hit%20Actor
    (:Is%20Valid
      (bind _returnvalue (GameplayTags|HasAnyMatchingGameplayTags(Message) Hit%20Actor (Variables|Default|GetBlacklisted%20Tags)))
      (return _returnvalue))
    (:Is%20Not%20Valid
      (return false))))

(fn CheckForRamp
  (params (Base%20Destination :type Vector))
  (bind _outhits (Collision|MultiSphereTraceByChannel :Start (+ Base%20Destination (LocalVariables|GetGrid%20Center%20Correction%20Offset)) :End (+ Base%20Destination (LocalVariables|GetGrid%20Center%20Correction%20Offset)) :Radius 5e+01 :TraceChannel "TraceTypeQuery10" :bTraceComplex false :DrawDebugType "None" :bIgnoreSelf true :TraceColor "(R=1.000000,G=0.224955,B=0.919216,A=1.000000)" :TraceHitColor "(R=0.951467,G=1.000000,B=0.862990,A=1.000000)" :DrawTime 5.0))
  (for _array_element _outhits
    (Utilities|IsValid (.HitActor (Collision|BreakHitResult _array_element))
      (:Is%20Valid
        (bind _returnvalue (GameplayTags|HasAnyMatchingGameplayTags(Message) (.HitActor (Collision|BreakHitResult _array_element)) (Variables|RampRules|GetRamp%20Gameplay%20Tags)))
        (if _returnvalue
          (bind _output_get (LocalVariables|SetHit%20Actor (.HitActor (Collision|BreakHitResult _array_element))))
          (bind _asbuilding_stairs (Utilities|Casting|CastToBuildingStairs _output_get)
            (:then
              (if (== (Building|GetNavigationLinksClass _asbuilding_stairs false) (LocalVariables|GetStairs%20Nav%20Links%20Class))
                _))
            (:CastFailed
              (bind _asbuilding_roof (Utilities|Casting|CastToBuildingRoof (LocalVariables|GetHit%20Actor)))
              (if (== (Building|GetNavigationLinksClass _asbuilding_roof false) (LocalVariables|GetRoof%20Nav%20Links%20Class))
                _)))
          (return true (LocalVariables|GetHit%20Actor))))))
  (return false 0))

(fn CheckifHostile
  (params (Actor :type Actor%20Object%20Reference))
  (Utilities|IsValid Actor
    (:Is%20Valid
      (bind _returnvalue (PlayerInfo|GetOwnerPlayerController))
      (if (Utilities|Enum|Equal(Enum) (Team|GetTeamAffiliation _returnvalue Actor) "Hostile")
        (return true)
        (else
          (return false))))))

(fn ClearAndBuild
  (params (ClearedGridCells :type Array%20of%20Vectors) (BGAClassArray :type Array%20of%20Actor%20Class%20References) (BGALocationArray :type Array%20of%20Vectors) (BGARotationArray :type Array%20of%20Rotators) (DestroyFloorsQuickly :type Boolean))
  (for _array_element ClearedGridCells
    (Tower|DestroyActorsinGridCellLocation (Building|GetBuildingStructuralSupportSystem) (+ (Variables|BuildingAndClearing|GetBaseDestination) (Math|Vector|RotateVectorAroundAxis _array_element (Utilities|Map|Find (Variables|Orientation|GetOrientationYawOffsets) (Variables|Orientation|GetCardinalDirection)) "0, 0,1.000000")) DestroyFloorsQuickly))
  (CallFunction|CreateBaseSection (Variables|BuildingAndClearing|GetBaseDestination) BGAClassArray BGALocationArray BGARotationArray))

(fn ClearOnly
  (params (ClearedGridCells :type Array%20of%20Vectors) (DestroyFloorsQuickly :type Boolean))
  (bind _buildingstructuralsupportsystem (Building|GetBuildingStructuralSupportSystem))
  (Building|StartActorRemovalBatch _buildingstructuralsupportsystem)
  (for _array_element ClearedGridCells
    (Tower|DestroyActorsinGridCellLocation (Building|GetBuildingStructuralSupportSystem) (+ (Variables|BuildingAndClearing|GetBaseDestination) (Math|Vector|RotateVectorAroundAxis _array_element (Utilities|Map|Find (Variables|Orientation|GetOrientationYawOffsets) (Variables|Orientation|GetCardinalDirection)) "0, 0,1.000000")) DestroyFloorsQuickly))
  (Building|StopActorRemovalBatch _buildingstructuralsupportsystem))

(fn CreateBaseSection
  (params (ReferenceLocation :type Vector) (BGAClassArray :type Array%20of%20Actor%20Class%20References) (BGALocationArray :type Array%20of%20Vectors) (BGARotationArray :type Array%20of%20Rotators))
  (if (> (Utilities|Array|Length BGAClassArray) 0)
    (if (Creative|Budget|CanPlaceClassInstances (Fortnite|GetFortPlayerControllerfromActor (Actor|GetOwner)) BGAClassArray ReferenceLocation)
      (for _index (range (Utilities|Array|Length BGAClassArray))
        (bind _returnvalue (Game|SpawnActorfromClass (Utilities|Array|Get(acopy) BGAClassArray _index) (Math|Transform|MakeTransform (Math|Vector|vector+vector ReferenceLocation (Math|Vector|RotateVectorAroundAxis (Utilities|Array|Get(aref) BGALocationArray _index) (Utilities|Map|Find (Variables|Orientation|GetOrientationYawOffsets) (Variables|Orientation|GetCardinalDirection)) "0, 0,1.000000") (Utilities|Map|Find (Variables|Orientation|GetOrientationLocationOffsets) (Variables|Orientation|GetCardinalDirection))) (Math|Rotator|MakeRotator 0.0 0.0 (+ (.yaw (Utilities|Array|Get(aref) BGARotationArray _index)) (Utilities|Map|Find (Variables|Orientation|GetOrientationYawOffsets) (Variables|Orientation|GetCardinalDirection))))) "Undefined" "SelectDefaultAtRuntime" 0 0))
        (bind _asbuilding_static_mesh (Utilities|Casting|CastToBuildingSMActor _returnvalue))
        (Class|BuildingActor|SetTeam _asbuilding_static_mesh (Team|GetActorTeam (Actor|GetOwner)))
        (Building|InitializeKismetSpawnedBuildingActor _asbuilding_static_mesh 0 0 true 0 false)
        (bind _asbuilding_trap (Utilities|Casting|CastToBuildingTrap _asbuilding_static_mesh)
          (:then
            (Trap|SpawnedfromItemDefinition _asbuilding_trap "/Game/Athena/Items/Traps/DamageTrap/TID_Floor_Spikes_Athena_R_T01.TID_Floor_Spikes_Athena_R_T01" 1))
          (:CastFailed
            _))
        (if (not (and (Building|IsWorldLocValid (Building|GetBuildingStructuralSupportSystem) (+ (Transformation|GetActorLocation _asbuilding_static_mesh) "0, 0,384.000000")) (Building|IsWorldLocValid (Building|GetBuildingStructuralSupportSystem) (+ (Transformation|GetActorLocation _asbuilding_static_mesh) "0, 0,-384.000000"))))
          (Actor|DestroyActor _asbuilding_static_mesh))))))

(fn FloorTrace
  (bind _outhits (Collision|MultiBoxTraceByChannel (+ (Variables|TerrainTrace|GetLaunchpadTraceOffsetLow) (Variables|BuildingandClearing|GetCenterDestination)) (Variables|BuildingandClearing|GetCenterDestination) "128.000000,128.000000, 0" "0, 0, 0" "TraceTypeQuery10"))
  (for _array_element _outhits
    (bind _returnvalue (GameplayTags|HasAnyMatchingGameplayTags(Message) (.HitActor (Collision|BreakHitResult _array_element)) (Variables|Default|GetFloor%20Gameplay%20Tag)))
    (if _returnvalue
      (return true (.HitActor (Collision|BreakHitResult _array_element)))))
  (return false 0))

(fn HandleRamps
  (params (Ramp%20Actor :type Actor%20Object%20Reference))
  (Utilities|IsValid Ramp%20Actor
    (:Is%20Valid
      (if (not (RampRules|RampUndersideCheck (Variables|Default|GetStopped%20Hit)))
        (Variables|BuildingAndClearing|SetBaseDestination (+ (Variables|BuildingAndClearing|GetBaseDestination) (Utilities|Map|Find (Variables|RampRules|GetRamp%20Rule%20Build%20Offsets) (RampRules|RoundToNearestRotation (Math|Float|Truncate (.yaw (Transformation|GetActorRotation Ramp%20Actor)))))))))))

(fn HandleProps
  (params (Hit%20Results :type Array%20of%20Actor%20Object%20References))
  (for _array_element Hit%20Results
    (Utilities|Casting|CastToBuildingProp _array_element
      (:then
        (Utilities|Casting|CastToTiered_Chest_Athena _array_element
          (:then
            _)
          (:CastFailed
            (Utilities|Casting|CastToBuildingPropBigHealth _array_element
              (:then
                _)
              (:CastFailed
                (bind _returnvalue (CallFunction|CanDestroyObject _array_element))
                (if _returnvalue
                  (Utilities|Array|Add (Variables|Tower|GetBuildingActorsToDestroy) _array_element)
                  (else
                    _))))
            (Utilities|Array|Add (Variables|Tower|GetBuildingActorsToIgnore) _array_element)
            ;[0.3854 0.3854 0.3854] TEMP, pending refactor
            ; needs to add other launchpads to destroy, to prevent stacking
            (bind _astrap_launch_pad_thrown (Utilities|Casting|CastToTrap_LaunchPadThrown _array_element))
            (Utilities|Array|Add (Variables|Tower|GetBuildingActorsToDestroy) _astrap_launch_pad_thrown))))
      (:CastFailed
        (Utilities|Casting|CastToBP_DamageBalloon_Athena _array_element
          (:then
            _)
          (:CastFailed
            (Utilities|Casting|CastToBuildingActor _array_element)))))))

(fn InstigatorStandingOnRamp
  (params (Ramp%20Actor :type Actor%20Object%20Reference))
  (bind _outhits (Collision|MultiSphereTraceByChannel (- (Transformation|GetActorLocation (Variables|Actor|GetInstigator)) "0,0,75.000000") (- (Transformation|GetActorLocation (Variables|Actor|GetInstigator)) "0,0,75.000000") 25.0 "TraceTypeQuery10"))
  (for _array_element _outhits
    (if (== Ramp%20Actor (.HitActor (Collision|BreakHitResult _array_element)))
      (return true)))
  (return false))

(fn RampUndersideCheck
  (params (Hit%20Result :type Hit%20Result%20Structure))
  (bind (_bblockinghit _binitialoverlap _time _distance _location _impactpoint _normal _impactnormal) (Collision|BreakHitResult Hit%20Result))
  (if (< (.z _impactnormal) 0)
    (return true)
    (else
      (return false))))

(fn RejectBuilding
  (params (Use%20Actor%20Location :type Boolean))
  (bind (_bblockinghit _binitialoverlap _time _distance _location _impactpoint) (Collision|BreakHitResult (Variables|Default|GetStopped%20Hit)))
  (Utilities|FlowControl|Sequence
    (:then_0
      (Components|Activation|Deactivate (Variables|Cosmetic|GetBuild%20Particle))
      (Rendering|SetActorHiddenInGame true))
    (:then_1
      (Inventory|SpawnPickupInWorld "/LaunchPadItemGameplay/Items/WID/WID_Athena_LaunchPadThrown_Super.WID_Athena_LaunchPadThrown_Super" 1 (select Use%20Actor%20Location (Transformation|GetActorLocation) _impactpoint) "0.000000,0.000000,1.000000" 0 true true false -1 "Tossed" "TossedByPlayer"))
    (:then_2
      (Actor|SetLifeSpan 0.5))))

(fn RoundtoNearestRotation
  (params (Angle :type Integer))
  (bind _output_get (LocalVariables|SetIncoming%20Angle Angle))
  (if (Math|Integer|InRange(Integer) _output_get -5 5)
    (return 0)
    (elif (Math|Integer|InRange(Integer) (LocalVariables|GetIncoming%20Angle) 85 95)
      (return 90)
      (elif (Math|Integer|InRange(Integer) (LocalVariables|GetIncoming%20Angle) 175 185)
        (return 180)
        (elif (Math|Integer|InRange(Integer) (LocalVariables|GetIncoming%20Angle) -95 -85)
          (return -90))))))

(fn SecondTerrainTrace
  (bind _outhits (Collision|MultiBoxTraceByChannel (+ (+ (Variables|BuildingandClearing|GetCenterDestination) "0,0,384.000000") (Variables|TerrainTrace|GetLaunchpadTraceOffsetLow)) (+ (Variables|BuildingandClearing|GetCenterDestination) (Variables|TerrainTrace|GetLaunchpadTraceOffsetHigh)) "128.000000,128.000000, 0" "0, 0, 0" "TraceTypeQuery10" false (Variables|Tower|GetBuildingActorsToIgnore) "None" true "(R=1.000000,G=0.000000,B=0.872107,A=1.000000)" "(R=1.000000,G=0.991730,B=0.910392,A=1.000000)"))
  (for _array_element _outhits
    (Utilities|Casting|CastToBuildingActor (.HitActor (Collision|BreakHitResult _array_element))
      (:CastFailed
        (return true))))
  (return false))

(fn SetBaseDestination
  (params (Hit :type Hit%20Result%20Structure%20%28by%20ref%29))
  (bind (_bblockinghit _binitialoverlap _time _distance _location _impactpoint _normal _impactnormal) (Collision|BreakHitResult Hit))
  (Networking|SwitchHasAuthority
    (:Authority
      (Variables|BuildingAndClearing|SetBaseDestination (Fortnite|SnaptoFortBuildingGrid (+ _impactpoint (* _impactnormal 1)) (Math|Vector|MakeVector 0.0 0.0 (Variables|SnaptoGrid|GetGridZOffset)) (Math|Vector|MakeVector 0.0 (Variables|SnaptoGrid|GetGridYOffset)))))))

(fn ShouldBounceOffActor
  (params (Hit%20Actor :type Actor%20Object%20Reference) (Hit%20Component :type Primitive%20Component%20Object%20Reference))
  (bind (_asbuilding_prop _bsuccess) (Utilities|Casting|CastToBuildingProp Hit%20Actor))
  (bind (_asfort_athena_supply_drop _bsuccess_2) (Utilities|Casting|CastToFortAthenaSupplyDrop Hit%20Actor))
  (bind (_asfort_projectile_base _bsuccess_3) (Utilities|Casting|CastToFortProjectileBase Hit%20Actor))
  (bind (_asathena_supply_drop_balloon _bsuccess_4) (Utilities|Casting|CastToAthenaSupplyDropBalloon Hit%20Actor))
  (bind (_asfort_pawn _bsuccess_5) (Utilities|Casting|CastToFortPawn Hit%20Actor))
  (Utilities|IsValid Hit%20Actor
    (:Is%20Valid
      (Utilities|IsValid Hit%20Component
        (:Is%20Valid
          (return (Math|Boolean|ORBoolean (Utilities|Array|ContainsItem (Variables|/Script/Engine.ActorComponent|GetComponentTags Hit%20Component) (LocalVariables|GetBounce%20Tag)) (Utilities|Array|ContainsItem (Variables|/Script/Engine.Actor|GetTags Hit%20Actor) (LocalVariables|GetBounce%20Tag)) _bsuccess _bsuccess_5 _bsuccess_2 _bsuccess_4 _bsuccess_3)))
        (:Is%20Not%20Valid
          _)))
    (:Is%20Not%20Valid
      _))
  (return false))

(fn ConstructionScript
  (Networking|CanExecuteCosmeticEvents
    (:True
      (Components|Movement|SetUpdatedComponent (Variables|Components|GetRotatingMovement) (Variables|Components|GetMesh)))))
